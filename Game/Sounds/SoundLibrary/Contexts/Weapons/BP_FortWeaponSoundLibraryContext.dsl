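; Weapon sound library context that plays fire and foley sounds for Fort weapons.
; Non-charged fire events reuse an audio component, replacing it when missing, almost finished (watched MetaSound output) or given a different sound.
; New fire components receive rate, ammo, suppressor and sound-override parameters before playing, via native or Blueprint rollback initialization.
; Non-fire events spawn attached sounds with any wave overrides; charged fire events play nothing here.
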(fn CheckIsFireValid
  (params (EventTag :type Gameplay%20Tag%20Structure) (SoundLibraryContextEventInput :type Sound%20Library%20Context%20Event%20Input%20Structure%20%28by%20ref%29))
  (return (GameplayTags|MatchesTag EventTag "(TagName=\"SoundLibrary.Weapon.Fire\")") (not (GameplayTags|HasAnyTags (Utilities|Struct|BreakSoundLibraryContextEventInput SoundLibraryContextEventInput) (GameplayTags|MakeLiteralGameplayTagContainer "(GameplayTags=((TagName=\"SoundLibrary.Weapon.Fire.Charged\")))")))))

(fn GetWeaponFireAudioParameters
  (params (Weapon :type Fort%20Weapon%20Object%20Reference) (EventTag :type Gameplay%20Tag%20Structure))
  (return (Utilities|Array|MakeArray (Audio|Parameter|FloattoAudioParameter "Weapon.FireRate" (Math|Float|SafeDivide 1.0 (FortWeapon|GetFiringRate Weapon))) (Audio|Parameter|IntegertoAudioParameter "Weapon.AmmoRemaining" (FortWeapon|GetMagazineAmmoCount Weapon)) (Audio|Parameter|BooleantoAudioParameter "Weapon.IsSuppressorEnabled" (FortWeapon|IsSuppressed Weapon)) (Audio|Parameter|ObjectArraytoAudioParameter "Sounds" (Weapon|GetSoundWaveOverrides EventTag)))))

(fn IsBurstFire
  (return (Variables|WeaponSoundLibraryContext|GetbIsBurstFire)))

(fn Play
  (params (InEventData_Sound :type Sound%20Base%20Object%20Reference) (InEventData_Owner :type Actor%20Object%20Reference) (InEventData_EventTag :type Gameplay%20Tag%20Structure) (InEventData_LibraryTags :type Gameplay%20Tag%20Container%20Structure))
  (bind _ineventdata (Variables|GetInEventData))
  (bind (_sound _owner _eventtag) (Utilities|Struct|BreakSoundLibraryContextEventInput _ineventdata))
  (bind _ineventdata_2 (Variables|GetInEventData))
  (bind (_sound_2 _owner_2 _eventtag_2) (Utilities|Struct|BreakSoundLibraryContextEventInput _ineventdata_2))
  (bind _returnvalue (Weapon|GetSoundWaveOverrides _eventtag))
  (bind _returnvalue_2 (Weapon|GetAttachComponent _owner _eventtag))
  (bind _returnvalue_3 (Weapon|GetAttachComponent _owner_2 _eventtag_2))
  (bind (_ineventdata_sound _ineventdata_owner _ineventdata_eventtag) (Utilities|Struct|BreakSoundLibraryContextEventInput (Variables|GetInEventData_Sound)))
  (bind (_matchesfiretag _canplaycharged) (WeaponSoundLibraryContext|CheckIsFireValid InEventData_EventTag (Variables|GetInEventData)))
  (if _matchesfiretag
    (if _canplaycharged
      ;[0.00967 0.03674 0.1042] Weapon Fire — We try to reuse audio components as an optimization, but need to spawn a new one if the existing one is almost finished. If we don't, then we risk sending parameters to an audio component that is about to be destroyed. Attached to the Player Pawn's mesh component
      (if (WeaponSoundLibraryContext|ShouldSpawnNewAudioComponent (Variables|GetInEventData))
        (Utilities|FlowControl|Sequence
          (:then_0
            (if (Weapon|ShouldUseNativeAudioComponentInitialization)
              (bind _returnvalue_4 (Weapon|InitializeWeaponFireAudioComponent (Variables|GetInEventData)))
              (Variables|WeaponSoundLibraryContext|SetWeapon%20Fire%20Audio%20Component _returnvalue_4)
              (Variables|WeaponSoundLibraryContext|GetWeapon%20Fire%20Audio%20Component
                (:then
                  (bind _actor (Variables|SoundLibrary|GetActor)
                    (:then
                      (bind _asfort_weapon (Utilities|Casting|CastToFortWeapon _actor))
                      (Audio|Parameter|SetParameters (Variables|WeaponSoundLibraryContext|GetWeapon%20Fire%20Audio%20Component) (WeaponSoundLibraryContext|GetWeaponFireAudioParameters _asfort_weapon _ineventdata_eventtag))
                      (Audio|Components|Audio|Play (Variables|WeaponSoundLibraryContext|GetWeapon%20Fire%20Audio%20Component) 0.0))
                    (:else)))
                (:else
                  _))
              (else
                ;;[0.4792 0.2047 0 0.5] Rollback: ShouldUseNativeAudioComponentInitialization
                (Utilities|IsValid _owner_2
                  (:Is%20Valid
                    (Utilities|IsValid _returnvalue_3
                      (:Is%20Valid
                        (bind _audiocomponent (WeaponSoundLibraryContext|PlayWeaponAudioComponent _sound_2 _returnvalue_3 _owner_2 _eventtag_2))
                        (bind _output_get (Variables|WeaponSoundLibraryContext|SetWeapon%20Fire%20Audio%20Component _audiocomponent))
                        (Utilities|IsValid _output_get
                          (:Is%20Not%20Valid
                            (Utilities|Debugging|RaiseScriptError "Failed to initialize weapon fire audio component!")
                            (return 0 false))))
                      (:Is%20Not%20Valid
                        _)))
                  (:Is%20Not%20Valid
                    _))
                (return 0 false))))
          (:then_1
            (Utilities|FlowControl|Sequence
              (:then_0
                (MetaSoundOutput|WatchOutput (WorldSubsystems|GetMetaSoundOutputSubsystem) (Variables|WeaponSoundLibraryContext|GetWeapon%20Fire%20Audio%20Component) "bIsAlmostFinished" (EventDispatchers|CreateEvent :!event Set%20Is%20Almost%20Finished) "None" "None" :!class /Script/MetasoundEngine.MetaSoundOutputSubsystem)
                (Variables|WeaponSoundLibraryContext|SetbIsBurstFire (GameplayTags|HasAnyTags (Utilities|Struct|BreakSoundLibraryContextEventInput (Variables|GetInEventData)) (GameplayTags|MakeLiteralGameplayTagContainer "(GameplayTags=((TagName=\"SoundLibrary.Weapon.Fire.Burst\")))")))
                (return (Utilities|Array|MakeArray (Variables|WeaponSoundLibraryContext|GetWeapon%20Fire%20Audio%20Component)) true)))))
        (else
          (return (Utilities|Array|MakeArray (Variables|WeaponSoundLibraryContext|GetWeapon%20Fire%20Audio%20Component)) true))))
    (else
      ;[0 0.1823 0.1491] Weapon Foley - attached to the Weapon root component or the Player Pawn's mesh component if its an Equip sound
      (Utilities|IsValid _owner
        (:Is%20Valid
          (Utilities|IsValid _returnvalue_2
            (:Is%20Valid
              (bind _returnvalue_5 (Audio|SpawnSoundAttached _sound _returnvalue_2))
              (if (Utilities|Array|IsNotEmpty _returnvalue)
                (Audio|Parameter|SetObjectArrayParameter _returnvalue_5 "Sounds" _returnvalue)
                (else
                  _))
              (return (Utilities|Array|MakeArray _returnvalue_5) true))))))))

(fn PlayWeaponAudioComponent
  (params (Sound :type Sound%20Base%20Object%20Reference) (AttachmentComponent :type Scene%20Component%20Object%20Reference) (Actor :type Actor%20Object%20Reference) (EventTag :type Gameplay%20Tag%20Structure))
  (bind _returnvalue (Actor|GetOwner Actor))
  ;[1 1 1] Adds and attach an Audio Component so parameters can be set before first playback
  (Utilities|IsValid _returnvalue
    (:Is%20Valid
      (bind _returnvalue_2 (Game|AddComponentbyClass _returnvalue "/Script/Engine.AudioComponent" true (Math|Transform|MakeTransform "0.000000,0.000000,60.000000")))
      (LocalVariables|SetAudio%20Component _returnvalue_2)
      (bind _audio_component (LocalVariables|GetAudio%20Component)
        (:then
          (Transformation|AttachComponentToComponent _audio_component (Variables|GetAttachmentComponent) "None" "KeepRelative" "KeepRelative" "KeepRelative" true)
          (Audio|Components|Audio|SetSound (LocalVariables|GetAudio%20Component) (Variables|GetSound))
          (Weapon|SetAutoDestroy (LocalVariables|GetAudio%20Component))
          (bind _asfort_weapon (Utilities|Casting|CastToFortWeapon (Variables|SoundLibrary|GetActor))
            (:then
              (Utilities|IsValid _asfort_weapon
                (:Is%20Valid
                  (Audio|Parameter|SetParameters (LocalVariables|GetAudio%20Component) (WeaponSoundLibraryContext|GetWeaponFireAudioParameters _asfort_weapon (Variables|GetEventTag)))
                  (Audio|Components|Audio|Play (LocalVariables|GetAudio%20Component) 0.0)
                  (return (LocalVariables|GetAudio%20Component)))))
            (:CastFailed
              (Audio|Components|Audio|Play (LocalVariables|GetAudio%20Component) 0.0)
              (return (LocalVariables|GetAudio%20Component)))))
        (:else)))))

(fn SetIsAlmostFinished
  (params (OutputName :type Name) (Output :type Meta%20Sound%20Output%20Structure%20%28by%20ref%29))
  (bind (_success _returnvalue) (MetaSoundOutput|GetBool Output))
  (Variables|WeaponSoundLibraryContext|SetIs%20Almost%20Finished _returnvalue))

(fn SetWeaponFireParameters
  (params (Weapon :type Fort%20Weapon%20Object%20Reference)))

(fn ShouldSpawnNewAudioComponent
  (params (In%20Event%20Data :type Sound%20Library%20Context%20Event%20Input%20Structure))
  (Utilities|FlowControl|Sequence
    (:then_0
      (Variables|WeaponSoundLibraryContext|GetWeapon%20Fire%20Audio%20Component
        (:then)
        (:else
          (return true))))
    (:then_1
      (Utilities|FlowControl|Sequence
        (:then_0
          (if (Variables|WeaponSoundLibraryContext|GetIs%20Almost%20Finished)
            (return true)))
        (:then_1
          (Utilities|FlowControl|Sequence
            (:then_0
              (if (!= (Variables|/Script/Engine.AudioComponent|GetSound (Variables|WeaponSoundLibraryContext|GetWeapon%20Fire%20Audio%20Component)) (Utilities|Struct|BreakSoundLibraryContextEventInput (Variables|GetIn%20Event%20Data)))
                (return true)))))))))
