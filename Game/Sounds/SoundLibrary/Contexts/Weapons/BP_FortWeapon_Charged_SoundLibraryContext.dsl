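; Sound library context for charged Fortnite weapons that supplies the check for whether a fire sound event is valid.
; CheckIsFireValid returns two results: whether the event tag matches SoundLibrary.Weapon.Fire, and whether the event input carries the SoundLibrary.Weapon.Fire.Charged tag.
; The check relies purely on gameplay tag queries, with no other state or logic in the file.
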(fn CheckIsFireValid
  (params (EventTag :type Gameplay%20Tag%20Structure) (SoundLibraryContextEventInput :type Sound%20Library%20Context%20Event%20Input%20Structure%20%28by%20ref%29))
  (return (GameplayTags|MatchesTag EventTag "(TagName=\"SoundLibrary.Weapon.Fire\")") (GameplayTags|HasAnyTags (Utilities|Struct|BreakSoundLibraryContextEventInput SoundLibraryContextEventInput) (GameplayTags|MakeLiteralGameplayTagContainer "(GameplayTags=((TagName=\"SoundLibrary.Weapon.Fire.Charged\")))"))))
