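; Sound library component for ranged weapons that controls the weapon's firing and scope canting audio.
; It caches the audio component played for fire events and drives its MetaSound parameters: single-shot or burst triggers, with looping fire using a bool in 1P and start/stop triggers in 3P.
; Unequip, pawn death and EndPlay stop bursts and all events.
; For the locally viewed pawn, canting changes swap first- and third-person sound events.
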(event EventBeginPlay
  (|CallParentFunction)
  (Utilities|FlowControl|Sequence
    (:then_0
      (bind _asfort_weapon_ranged (Utilities|Casting|CastToFortWeaponRanged (Components|GetOwner)))
      (bind _output_get (Variables|WeaponSoundLibraryComponent|SetWeapon _asfort_weapon_ranged))
      (Targeting|BindEventtoOnScopedThirdPersonTargetingChanged _output_get (EventDispatchers|CreateEvent :!event Custom|Canting%20Mode%20Changed)))
    (:then_1
      (Utilities|FlowControl|Sequence
        (:then_0
          (bind _weapon (Variables|WeaponSoundLibraryComponent|GetWeapon)
            (:then
              (bind _asfort_player_pawn_athena (Utilities|Casting|CastToFortPlayerPawnAthena (Game|GetInstigator _weapon)))
              (bind _output_get_2 (Variables|WeaponSoundLibraryComponent|SetPlayer%20Pawn _asfort_player_pawn_athena))
              (FortPawn|AssignOnWeaponUnEquippedDelegate _output_get_2 (EventDispatchers|CreateEvent :!event Custom|Weapon%20Unequipped))
              (EventDispatchers|BindEventtoOnDeathEffects _output_get_2 (EventDispatchers|CreateEvent :!event Custom|Death%20Effects)))
            (:else)))
        (:then_1
          (Utilities|FlowControl|Sequence
            (:then_0
              (bind _asbp_fort_weapon_sound_library_context (Utilities|Casting|CastToBP_FortWeaponSoundLibraryContext (SoundLibrary|GetContext)))
              (Variables|WeaponSoundLibraryComponent|SetWeapon%20Sound%20Library%20Context _asbp_fort_weapon_sound_library_context))))))))

(event Weapon|EventPlaySingleFireSFX
  ; 1p supports firing single shots, which are handled using a trigger input on the MetaSound. Non-persistent fire weapons will fire all shots using this trigger.
  (bind _weapon_fire_audio_component (Variables|WeaponSoundLibraryComponent|GetWeapon%20Fire%20Audio%20Component)
    (:then
      (Audio|Parameter|ExecuteTriggerParameter _weapon_fire_audio_component "FireSingleShot")
      (bind _weapon_sound_library_context (Variables|WeaponSoundLibraryComponent|GetWeapon%20Sound%20Library%20Context)
        (:then
          (Variables|/Game/Sounds/SoundLibrary/Contexts/Weapons/BP_FortWeaponSoundLibraryContext.BP_FortWeaponSoundLibraryContext_C|SetIs%20Almost%20Finished false _weapon_sound_library_context))
        (:else)))
    (:else)))

(event Weapon|EventPlayLoopingFireSFX
  (bind _weapon_sound_library_context (Variables|WeaponSoundLibraryComponent|GetWeapon%20Sound%20Library%20Context)
    (:then
      (if (WeaponSoundLibraryContext|IsBurstFire _weapon_sound_library_context)
        (bind _weapon_fire_audio_component (Variables|WeaponSoundLibraryComponent|GetWeapon%20Fire%20Audio%20Component)
          (:then
            (Audio|Parameter|ExecuteTriggerParameter _weapon_fire_audio_component "FireBurst"))
          (:else))
        (elif (GameplayTags|MatchesTag (Variables|WeaponSoundLibraryComponent|GetIn%20Event%20Name) "(TagName=\"SoundLibrary.Weapon.Fire.1P\")")
          (bind _weapon_fire_audio_component_2 (Variables|WeaponSoundLibraryComponent|GetWeapon%20Fire%20Audio%20Component)
            (:then
              (Audio|Parameter|SetBooleanParameter _weapon_fire_audio_component_2 "bIsFiring" true :!class /Script/AudioExtensions.AudioParameterControllerInterface))
            (:else))
          (elif (GameplayTags|MatchesTag (Variables|WeaponSoundLibraryComponent|GetIn%20Event%20Name) "(TagName=\"SoundLibrary.Weapon.Fire.3P\")")
            (bind _weapon_fire_audio_component_3 (Variables|WeaponSoundLibraryComponent|GetWeapon%20Fire%20Audio%20Component)
              (:then
                (Audio|Parameter|ExecuteTriggerParameter _weapon_fire_audio_component_3 "StartFiring"))
              (:else)))))
      (Variables|/Game/Sounds/SoundLibrary/Contexts/Weapons/BP_FortWeaponSoundLibraryContext.BP_FortWeaponSoundLibraryContext_C|SetIs%20Almost%20Finished false (Variables|WeaponSoundLibraryComponent|GetWeapon%20Sound%20Library%20Context)))
    (:else)))

(event EventEndPlay
  (params (EndPlayReason :type EEndPlayReason%20Enum :default "Destroyed"))
  (bind _weapon_fire_audio_component (Variables|WeaponSoundLibraryComponent|GetWeapon%20Fire%20Audio%20Component)
    (:then
      (Audio|Parameter|ExecuteTriggerParameter _weapon_fire_audio_component "StopBurst")
      (SoundLibrary|StopAllEvents))
    (:else)))

(event Custom|Weapon%20Unequipped
  (bind _weapon_fire_audio_component (Variables|WeaponSoundLibraryComponent|GetWeapon%20Fire%20Audio%20Component)
    (:then
      (Audio|Parameter|ExecuteTriggerParameter _weapon_fire_audio_component "StopBurst")
      (SoundLibrary|StopAllEvents))
    (:else)))

(event Custom|Death%20Effects
  (params (DeadActor :type Actor%20Object%20Reference) (Damage :type Float%20%28single-precision%29) (InstigatedBy :type Fort%20Pawn%20Object%20Reference) (HitLocation :type Vector) (DamageCauser :type Actor%20Object%20Reference))
  (bind _weapon_fire_audio_component (Variables|WeaponSoundLibraryComponent|GetWeapon%20Fire%20Audio%20Component)
    (:then
      (Audio|Parameter|ExecuteTriggerParameter _weapon_fire_audio_component "StopBurst")
      (SoundLibrary|StopAllEvents))
    (:else)))

(event Custom|Canting%20Mode%20Changed
  (params (bNewIsScopedThirdPersonTargeting :type Boolean))
  ; Canting — feature that allows scoped weapons to aim down sights using a third person perspective instead of a first person perspective
  (bind _player_pawn (Variables|WeaponSoundLibraryComponent|GetPlayer%20Pawn)
    (:then
      (if (Pawn|IsLocallyViewed _player_pawn)
        (Utilities|FlowControl|Sequence
          (:then_0
            (SoundLibrary|StopEvent (select bNewIsScopedThirdPersonTargeting "(TagName=\"SoundLibrary.Weapon.Canting.FirstPerson\")" "(TagName=\"SoundLibrary.Weapon.Canting.ThirdPerson\")")))
          (:then_1
            (Utilities|FlowControl|Sequence
              (:then_0
                (SoundLibrary|PlayEvent (select bNewIsScopedThirdPersonTargeting "(TagName=\"SoundLibrary.Weapon.Canting.ThirdPerson\")" "(TagName=\"SoundLibrary.Weapon.Canting.FirstPerson\")"))))))))
    (:else)))

(fn OnSoundPlayed
  (params (InEventName :type Gameplay%20Tag%20Structure) (InComponent :type Audio%20Component%20Object%20Reference))
  (|CallParentFunction InEventName InComponent)
  (bind _output_get (Variables|WeaponSoundLibraryComponent|SetIn%20Event%20Name InEventName))
  (if (GameplayTags|MatchesTag _output_get "(TagName=\"SoundLibrary.Weapon.Fire\")")
    (Variables|WeaponSoundLibraryComponent|SetWeapon%20Fire%20Audio%20Component (Variables|GetInComponent))
    (return)))

(fn OnSoundStopped
  (params (InEventName :type Gameplay%20Tag%20Structure) (InComponent :type Audio%20Component%20Object%20Reference))
  (if (GameplayTags|MatchesTag InEventName "(TagName=\"SoundLibrary.Weapon.Fire\")")
    (if (GameplayTags|MatchesTag (Variables|GetInEventName) "(TagName=\"SoundLibrary.Weapon.Fire.1P\")")
      (bind _incomponent (Variables|GetInComponent)
        (:then
          (Audio|Parameter|SetBooleanParameter _incomponent "bIsFiring" false :!class /Script/AudioExtensions.AudioParameterControllerInterface)
          (return true))
        (:else
          (return true)))
      (elif (GameplayTags|MatchesTag (Variables|GetInEventName) "(TagName=\"SoundLibrary.Weapon.Fire.3P\")")
        (bind _incomponent_2 (Variables|GetInComponent)
          (:then
            (Audio|Parameter|ExecuteTriggerParameter _incomponent_2 "StopFiring")
            (return true))
          (:else
            (return true)))))
    (else
      (return false))))
