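; Weapon sound library component that manages gunfire audio, scoped-targeting canting foley, and stopping fire on death, unequip or EndPlay.
; Fire playback pushes suppressor and fire-rate parameters when no parameter pack is used; stopping triggers OnStop.
; Builder functions chain base gunfire and mod patches into runtime 1P 5.1 or 3P stereo MetaSounds for a priority library.
; Mod events only toggle the suppressor flag, a temporary performance change, without calling those builders.
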(event EventBeginPlay
  (|CallParentFunction)
  (bind _asfort_weapon_ranged (Utilities|Casting|CastToFortWeaponRanged (Components|GetOwner)))
  (Targeting|BindEventtoOnScopedThirdPersonTargetingChanged _asfort_weapon_ranged (EventDispatchers|CreateEvent :!event Custom|Scoped%20Third%20Person%20Targeting%20Changed))
  (bind _asfort_player_pawn_athena (Utilities|Casting|CastToFortPlayerPawnAthena (Game|GetInstigator _asfort_weapon_ranged)))
  (bind _output_get (Variables|Default|SetPlayer%20Pawn _asfort_player_pawn_athena))
  (EventDispatchers|BindEventtoOnDeathEffects _output_get (EventDispatchers|CreateEvent :!event StopFireEventsOnDeath))
  (FortPawn|AssignOnWeaponUnEquippedDelegate _output_get (EventDispatchers|CreateEvent :!event StopFireEventsOnUnequip)))

(event EventEndPlay
  (params (EndPlayReason :type EEndPlayReason%20Enum :default "Destroyed"))
  (CallFunction|StopFireEventsonDeath 0))

(event EventOnModApplied
  (params (InModAudioData :type Fort%20Weapon%20Mod%20Audio%20Data%20Object%20Reference))
  ;[0.2302 0.2302 0.2302] Temporarily changing for 28.00 perf
  (bind _bshouldbuild (Helper|ValidateModMetasounds (Utilities|ResolveSoftReference (Variables|/Script/FortniteGame.FortWeaponModAudioData|GetMetaSoundPatch1P InModAudioData)) (Utilities|ResolveSoftReference (Variables|/Script/FortniteGame.FortWeaponModAudioData|GetMetaSoundPatch3P InModAudioData))))
  (if _bshouldbuild
    (Variables|Default|SetbEnableSuppressor true)))

(event EventOnModRemoved
  (params (InModAudioData :type Fort%20Weapon%20Mod%20Audio%20Data%20Object%20Reference))
  ;[0.2302 0.2302 0.2302] Temporarily changing for 28.00 perf
  (bind _bshouldbuild (Helper|ValidateModMetasounds (Utilities|ResolveSoftReference (Variables|/Script/FortniteGame.FortWeaponModAudioData|GetMetaSoundPatch1P InModAudioData)) (Utilities|ResolveSoftReference (Variables|/Script/FortniteGame.FortWeaponModAudioData|GetMetaSoundPatch3P InModAudioData))))
  (if _bshouldbuild
    (Variables|Default|SetbEnableSuppressor false)))

(event Custom|Scoped%20Third%20Person%20Targeting%20Changed
  (params (bNewIsScopedThirdPersonTargeting :type Boolean))
  ; Weapon canting foley sounds
  (bind _player_pawn (Variables|Default|GetPlayer%20Pawn)
    (:then
      (if (Pawn|IsLocallyViewed _player_pawn)
        (Utilities|FlowControl|Sequence
          (:then_0
            (SoundLibrary|StopEvent (select bNewIsScopedThirdPersonTargeting "(TagName=\"SoundLibrary.Weapon.Canting.FirstPerson\")" "(TagName=\"SoundLibrary.Weapon.Canting.ThirdPerson\")")))
          (:then_1
            (Utilities|FlowControl|Sequence
              (:then_0
                (SoundLibrary|PlayEvent (select bNewIsScopedThirdPersonTargeting "(TagName=\"SoundLibrary.Weapon.Canting.ThirdPerson\")" "(TagName=\"SoundLibrary.Weapon.Canting.FirstPerson\")"))))))))
    (:else)))

(fn AddConnectedGraphInput
  (params (InputName :type Name) (DataType :type Name :default "Int32") (BuilderRef :type Meta%20Sound%20Source%20Builder%20Object%20Reference) (BaseWeaponNodeHandleRef :type Meta%20Sound%20Node%20Handle%20Structure)))

(fn BuiltSoundFireRateInit
  (params (BuilderRef :type Meta%20Sound%20Builder%20Base%20Object%20Reference))
  (bind _returnvalue (Components|GetOwner))
  (bind _asfort_weapon (Utilities|Casting|CastToFortWeapon _returnvalue))
  ;[1 1 1] Temp fix to prevent fire rate default overriding correct value. Setting default values manually
  (Utilities|IsValid _asfort_weapon
    (:Is%20Valid
      (bind (_datatype _returnvalue_2) (Audio|MetaSound|Builder|CreateMetaSoundFloatLiteral (EngineSubsystems|GetMetaSoundBuilderSubsystem) (select (Utilities|String|Contains (Utilities|GetObjectName _returnvalue) "SMG_Paprika_Burst") 0.064516 (Math|Float|SafeDivide 1.0 (FortWeapon|GetFiringRate _asfort_weapon)))))
      (Audio|MetaSound|Builder|SetGraphInputDefault BuilderRef "Weapon.FireRate" _returnvalue_2 ""))))

(fn CheckIfSuppressor
  (params (BuilderRef :type Meta%20Sound%20Builder%20Base%20Object%20Reference) (BaseNodeHandle :type Meta%20Sound%20Node%20Handle%20Structure%20%28by%20ref%29) (ModNodeHandle :type Meta%20Sound%20Node%20Handle%20Structure) (ModRef :type Object%20Reference))
  (if (Utilities|String|Contains (Utilities|GetObjectName ModRef) "Suppres")
    (Variables|Default|SetbSuppressorAtt true)
    (bind _returnvalue (Audio|MetaSound|Builder|FindNodeInputbyName BuilderRef BaseNodeHandle "MainFire.DistanceCurves")
      (:Succeeded
        (bind _returnvalue_2 (Audio|MetaSound|Builder|CreateMetaSoundObjectLiteral (EngineSubsystems|GetMetaSoundBuilderSubsystem) (Variables|Default|GetModDistanceCurveBank)))
        (Audio|MetaSound|Builder|SetNodeInputDefault BuilderRef _returnvalue _returnvalue_2
          (:Succeeded
            (ModBuilder|ConnectNodesIO BuilderRef "CloseCurveFloat" "CloseCurveEvaluation" BaseNodeHandle ModNodeHandle)
            (ModBuilder|ConnectNodesIO BuilderRef "MediumCurveFloat" "MediumCurveEvaluation" BaseNodeHandle ModNodeHandle)))))
    (else
      (Variables|Default|SetbSuppressorAtt false)
      (return))))

(fn ConnectInterfaceInputs
  (params (BuilderRef :type Meta%20Sound%20Builder%20Base%20Object%20Reference) (BaseWeaponNodeHandle :type Meta%20Sound%20Node%20Handle%20Structure%20%28by%20ref%29) (InputName :type Name))
  (bind _returnvalue (Audio|MetaSound|Builder|FindNodeInputbyName BuilderRef BaseWeaponNodeHandle InputName)
    (:Succeeded
      (Audio|MetaSound|Builder|ConnectNodeInputtoGraphInput BuilderRef InputName _returnvalue
        (:Succeeded
          (return))))))

(fn ConnectNodesIO
  (params (BuilderRef :type Meta%20Sound%20Builder%20Base%20Object%20Reference) (InputName :type Name) (OutputName :type Name) (BaseNodeHandle :type Meta%20Sound%20Node%20Handle%20Structure) (ModNodeHandle :type Meta%20Sound%20Node%20Handle%20Structure))
  (bind _returnvalue (Audio|MetaSound|Builder|FindNodeInputbyName BuilderRef ModNodeHandle InputName)
    (:Succeeded
      (bind _returnvalue_2 (Audio|MetaSound|Builder|FindNodeOutputbyName BuilderRef BaseNodeHandle OutputName)
        (:Succeeded
          (Audio|MetaSound|Builder|ConnectNodes BuilderRef _returnvalue_2 _returnvalue
            (:Succeeded
              (return))))))))

(fn CreateModMetaSound
  (bind _basegunfirems (Variables|Default|GetBaseGunfireMS)
    (:then
      _)
    (:else
      (Helper|SetBaseMS)))
  (Utilities|IsValid _basegunfirems
    (:Is%20Valid
      (bind (_onplaynodeoutput _onfinishednodeinput _audiooutnodeinputs _returnvalue) (Audio|MetaSound|Builder|CreateSourceBuilder (EngineSubsystems|GetMetaSoundBuilderSubsystem) "FortWeaponModMetasound" (select (CallFunction|IsLocallyViewed) "FiveDotOne" "Stereo") true)
        (:Succeeded
          (LocalVariables|SetSourceOnPlayNodeOutput _onplaynodeoutput)
          (LocalVariables|SetSourceOnFinishedNodeInput _onfinishednodeinput)
          (LocalVariables|SetMSSBuilder _returnvalue)
          (bind _returnvalue_2 (Audio|MetaSound|Builder|AddMetaSoundNodeFromAssetClass (LocalVariables|GetMSSBuilder) (Variables|Default|GetBaseGunfireMS))
            (:Succeeded
              (LocalVariables|SetBaseWeaponNodeHandle _returnvalue_2)
              (bind _returnvalue_3 (Audio|MetaSound|Builder|FindNodeInputbyName (LocalVariables|GetMSSBuilder) (LocalVariables|GetBaseWeaponNodeHandle) "UE.Source.OnPlay")
                (:Succeeded
                  (Audio|MetaSound|Builder|ConnectNodes (LocalVariables|GetMSSBuilder) (LocalVariables|GetSourceOnPlayNodeOutput) _returnvalue_3
                    (:Succeeded
                      (bind _returnvalue_4 (Audio|MetaSound|Builder|AddGraphInputNode :self (LocalVariables|GetMSSBuilder) :Name "OnStop" :DataType "Trigger" :bIsConstructorInput false)
                        (:Succeeded
                          (LocalVariables|SetSourceOnStopOutputHandle _returnvalue_4)
                          (bind _returnvalue_5 (Audio|MetaSound|Builder|FindNodeInputbyName (LocalVariables|GetMSSBuilder) (LocalVariables|GetBaseWeaponNodeHandle) "OnStop")
                            (:Succeeded
                              (Audio|MetaSound|Builder|ConnectNodes (LocalVariables|GetMSSBuilder) (LocalVariables|GetSourceOnStopOutputHandle) _returnvalue_5
                                (:Succeeded
                                  (Audio|MetaSound|Builder|AddInterface (LocalVariables|GetMSSBuilder) "Fort.Weapon"
                                    (:Succeeded
                                      (ModBuilder|ConnectInterfaceInputs (LocalVariables|GetMSSBuilder) (LocalVariables|GetBaseWeaponNodeHandle) "Weapon.AmmoRemaining")
                                      (ModBuilder|ConnectInterfaceInputs (LocalVariables|GetMSSBuilder) (LocalVariables|GetBaseWeaponNodeHandle) "Weapon.FireRate")
                                      (ModBuilder|ConnectInterfaceInputs (LocalVariables|GetMSSBuilder) (LocalVariables|GetBaseWeaponNodeHandle) "Weapon.IsAimingDownSights")
                                      (Audio|MetaSound|Builder|AddInterface (LocalVariables|GetMSSBuilder) "Fort.Character"
                                        (:Succeeded
                                          (ModBuilder|ConnectInterfaceInputs (LocalVariables|GetMSSBuilder) (LocalVariables|GetBaseWeaponNodeHandle) "Character.Affiliation")
                                          (ModBuilder|ConnectInterfaceInputs (LocalVariables|GetMSSBuilder) (LocalVariables|GetBaseWeaponNodeHandle) "Character.Health")
                                          (ModBuilder|ConnectInterfaceInputs (LocalVariables|GetMSSBuilder) (LocalVariables|GetBaseWeaponNodeHandle) "Character.UniqueID")
                                          (if (not (CallFunction|IsLocallyViewed))
                                            (Audio|MetaSound|Builder|AddInterface (LocalVariables|GetMSSBuilder) "UE.Attenuation"
                                              (:Succeeded
                                                (ModBuilder|ConnectInterfaceInputs (LocalVariables|GetMSSBuilder) (LocalVariables|GetBaseWeaponNodeHandle) "UE.Attenuation.Distance")
                                                (Audio|MetaSound|Builder|AddInterface (LocalVariables|GetMSSBuilder) "UE.Source.Orientation"
                                                  (:Succeeded
                                                    (ModBuilder|ConnectInterfaceInputs (LocalVariables|GetMSSBuilder) (LocalVariables|GetBaseWeaponNodeHandle) "UE.Source.Orientation.Azimuth")
                                                    (ModBuilder|ConnectInterfaceInputs (LocalVariables|GetMSSBuilder) (LocalVariables|GetBaseWeaponNodeHandle) "UE.Source.Orientation.Elevation")
                                                    (Audio|MetaSound|Builder|AddInterface (LocalVariables|GetMSSBuilder) "UE.Spatialization"
                                                      (:Succeeded
                                                        (ModBuilder|ConnectInterfaceInputs (LocalVariables|GetMSSBuilder) (LocalVariables|GetBaseWeaponNodeHandle) "UE.Spatialization.Azimuth")
                                                        (ModBuilder|ConnectInterfaceInputs (LocalVariables|GetMSSBuilder) (LocalVariables|GetBaseWeaponNodeHandle) "UE.Spatialization.Elevation"))))))))
                                          (for _array_element (Variables|Default|GetModMS)
                                            (bind _returnvalue_6 (Audio|MetaSound|Builder|AddMetaSoundNodeFromAssetClass (LocalVariables|GetMSSBuilder) _array_element)
                                              (:Succeeded
                                                (LocalVariables|SetModNodeHandle _returnvalue_6)
                                                (bind _returnvalue_7 (Audio|MetaSound|Builder|FindNodeOutputbyName (LocalVariables|GetMSSBuilder) (LocalVariables|GetBaseWeaponNodeHandle) "OnGunshotOut")
                                                  (:Succeeded
                                                    (LocalVariables|SetBaseWeaponGunshotOutHandle _returnvalue_7)
                                                    (bind _returnvalue_8 (Audio|MetaSound|Builder|FindNodeInputbyName (LocalVariables|GetMSSBuilder) (LocalVariables|GetModNodeHandle) "OnGunshot")
                                                      (:Succeeded
                                                        (Audio|MetaSound|Builder|ConnectNodes (LocalVariables|GetMSSBuilder) (LocalVariables|GetBaseWeaponGunshotOutHandle) _returnvalue_8
                                                          (:Succeeded
                                                            (if (CallFunction|IsLocallyViewed)
                                                              (ModBuilder|ConnectNodesIO (LocalVariables|GetMSSBuilder) "UE.InputFormat.5dot1.Audio:0" "UE.OutputFormat.5dot1.Audio:0" (LocalVariables|GetBaseWeaponNodeHandle) (LocalVariables|GetModNodeHandle))
                                                              (ModBuilder|ConnectNodesIO (LocalVariables|GetMSSBuilder) "UE.InputFormat.5dot1.Audio:1" "UE.OutputFormat.5dot1.Audio:1" (LocalVariables|GetBaseWeaponNodeHandle) (LocalVariables|GetModNodeHandle))
                                                              (ModBuilder|ConnectNodesIO (LocalVariables|GetMSSBuilder) "UE.InputFormat.5dot1.Audio:2" "UE.OutputFormat.5dot1.Audio:2" (LocalVariables|GetBaseWeaponNodeHandle) (LocalVariables|GetModNodeHandle))
                                                              (ModBuilder|ConnectNodesIO (LocalVariables|GetMSSBuilder) "UE.InputFormat.5dot1.Audio:3" "UE.OutputFormat.5dot1.Audio:3" (LocalVariables|GetBaseWeaponNodeHandle) (LocalVariables|GetModNodeHandle))
                                                              (ModBuilder|ConnectNodesIO (LocalVariables|GetMSSBuilder) "UE.InputFormat.5dot1.Audio:4" "UE.OutputFormat.5dot1.Audio:4" (LocalVariables|GetBaseWeaponNodeHandle) (LocalVariables|GetModNodeHandle))
                                                              (ModBuilder|ConnectNodesIO (LocalVariables|GetMSSBuilder) "UE.InputFormat.5dot1.Audio:5" "UE.OutputFormat.5dot1.Audio:5" (LocalVariables|GetBaseWeaponNodeHandle) (LocalVariables|GetModNodeHandle))
                                                              (else
                                                                (ModBuilder|ConnectNodesIO (LocalVariables|GetMSSBuilder) "UE.InputFormat.Stereo.Audio:0" "UE.OutputFormat.Stereo.Audio:0" (LocalVariables|GetBaseWeaponNodeHandle) (LocalVariables|GetModNodeHandle))
                                                                (ModBuilder|ConnectNodesIO (LocalVariables|GetMSSBuilder) "UE.InputFormat.Stereo.Audio:1" "UE.OutputFormat.Stereo.Audio:1" (LocalVariables|GetBaseWeaponNodeHandle) (LocalVariables|GetModNodeHandle))
                                                                (ModBuilder|CheckifSuppressor (LocalVariables|GetMSSBuilder) (LocalVariables|GetBaseWeaponNodeHandle) (LocalVariables|GetModNodeHandle) _array_element))))))))))))
                                          (Audio|MetaSound|Builder|ConnectNodeOutputstoMatchingGraphInterfaceOutputs (LocalVariables|GetMSSBuilder) (LocalVariables|GetModNodeHandle)
                                            (:Succeeded
                                              (bind _returnvalue_9 (Audio|MetaSound|Builder|FindNodeOutputbyName (LocalVariables|GetMSSBuilder) (LocalVariables|GetBaseWeaponNodeHandle) "UE.Source.OneShot.OnFinished")
                                                (:Succeeded
                                                  (Audio|MetaSound|Builder|ConnectNodes (LocalVariables|GetMSSBuilder) _returnvalue_9 (LocalVariables|GetSourceOnFinishedNodeInput)
                                                    (:Succeeded
                                                      (Helper|BuiltSoundFireRateInit (LocalVariables|GetMSSBuilder))
                                                      (bind _asmetasound_source (Utilities|Casting|CastToMetaSoundSource (Audio|MetaSound|Builder|BuildNewMetaSound (LocalVariables|GetMSSBuilder) (Utilities|String|StringToName (Utilities|String|Append (Utilities|GetObjectName (Variables|Default|GetBaseGunfireMS)) "_Mod")))))
                                                      (bind _asfort_weapon_sound_library_context (Utilities|Casting|CastToFortWeaponSoundLibraryContext (SoundLibrary|GetContext)))
                                                      (Utilities|IsValid _asfort_weapon_sound_library_context
                                                        (:Is%20Valid
                                                          (Weapon|ConfigureMetaSoundSource _asfort_weapon_sound_library_context _asmetasound_source (select (CallFunction|IsLocallyViewed) "/Game/Sounds/FortSoundClassesAndModes/Fort_WeaponRanged_Fire_1P_PSM.Fort_WeaponRanged_Fire_1P_PSM" "/Game/Sounds/FortSoundClassesAndModes/Fort_WeaponRanged_Fire_3P_PSM.Fort_WeaponRanged_Fire_3P_PSM") (select (CallFunction|IsLocallyViewed) 0 (select (Variables|Default|GetbSuppressorAtt) "/Game/Sounds/Attenuation/Fort_Weapon_Projectile_NonAtt_Suppressed_Athena_Att.Fort_Weapon_Projectile_NonAtt_Suppressed_Athena_Att" "/Game/Sounds/Attenuation/Fort_Weapon_Projectile_NonAtt_Athena_Att.Fort_Weapon_Projectile_NonAtt_Athena_Att")) (select (CallFunction|IsLocallyViewed) (Utilities|Set|MakeSet "/Game/Sounds/Concurrency_Rules/Weapon_Gunfire_Local_Conc.Weapon_Gunfire_Local_Conc") (Utilities|Set|MakeSet "/Game/Sounds/Concurrency_Rules/Weapon_Gunfire_Local_Conc.Weapon_Gunfire_Local_Conc" "/Game/Sounds/Concurrency_Rules/Weapon_Gunfire_Global_Conc.Weapon_Gunfire_Global_Conc")))
                                                          (CallFunction|CacheMetaSound _asmetasound_source))))))))))))))))))))))))))))))

(fn CreateModSoundLibrary
  (bind _returnvalue (CallFunction|GetCachedMetaSound))
  (bind _modsoundlibrary (Variables|Default|GetModSoundLibrary))
  (Variables|Default|GetModSoundLibrary
    (:then
      _)
    (:else
      (bind _returnvalue_2 (Game|ConstructObjectfromClass "/Script/SoundLibrary.SoundLibrary"))
      (Variables|Default|SetModSoundLibrary _returnvalue_2)
      (Utilities|IsValid _returnvalue
        (:Is%20Valid
          (Scripting|AddSound(Transient) _modsoundlibrary (select (CallFunction|IsLocallyViewed) "(TagName=\"SoundLibrary.Weapon.Fire.1P\")" "(TagName=\"SoundLibrary.Weapon.Fire.3P\")") _returnvalue)
          (Scripting|SetChannel _modsoundlibrary "(TagName=\"SoundLibrary.Weapon\")")
          (Scripting|SetPriority _modsoundlibrary 10)))))
  (SoundLibrary|AddLibrary _modsoundlibrary))

(fn GetIsRetrigger
  (bind _returnvalue (SoundLibrary|GetContext))
  (Utilities|IsValid _returnvalue
    (:Is%20Valid
      (bind _asb_weapon_sound_library_context (Utilities|Casting|CastToB_WeaponSoundLibraryContext _returnvalue)
        (:then
          (return (not (Variables|/Game/Sounds/SoundLibrary/Contexts/Weapons/B_WeaponSoundLibraryContext.B_WeaponSoundLibraryContext_C|GetbAlmostFinished _asb_weapon_sound_library_context))))
        (:CastFailed
          _)))
    (:Is%20Not%20Valid
      _))
  (return true))

(fn GetSoundsFromEvent
  (params (InEventName :type Gameplay%20Tag%20Structure))
  (Utilities|FlowControl|Sequence
    (:then_0
      (bind _asb_weapon_sound_library_context (Utilities|Casting|CastToB_WeaponSoundLibraryContext (SoundLibrary|GetContext)))
      (bind _output_get (LocalVariables|SetWeaponContext _asb_weapon_sound_library_context))
      (Class|BWeaponSoundLibraryContext|EnableSoundRetrieval _output_get))
    (:then_1
      (SoundLibrary|PlayEvent InEventName) ; This won't do anything but cache the USoundBase assets into an internal array within the context
      (LocalVariables|SetSounds (Class|BWeaponSoundLibraryContext|GetRetrievedSounds (LocalVariables|GetWeaponContext))))
    (:then_2
      (Class|BWeaponSoundLibraryContext|ClearRetrievedSounds (LocalVariables|GetWeaponContext)))
    (:then_3
      (return (LocalVariables|GetSounds)))))

(fn OnSoundPlayed
  (params (InEventName :type Gameplay%20Tag%20Structure) (InComponent :type Audio%20Component%20Object%20Reference))
  (|CallParentFunction InEventName InComponent)
  (if (not (Weapon|IsUsingMetaSoundParameterPack))
    (Helper|TryHandleFireEvent InEventName InComponent))
  ;[0 0 0] Forces correct MS parameters on switch
  (if (GameplayTags|MatchesTag InEventName (Variables|Default|GetFireEvent))
    (if (not (Audio|Components|Audio|IsPlaying InComponent))
      (Audio|Components|Audio|Play InComponent 0.0))))

(fn OnSoundStopped
  (params (InEventName :type Gameplay%20Tag%20Structure) (InComponent :type Audio%20Component%20Object%20Reference))
  (if (GameplayTags|MatchesTag InEventName (Variables|Default|GetFireEvent))
    (Audio|Parameter|ExecuteTriggerParameter InComponent "OnStop")
    (return true)
    (else
      (return false))))

(fn SetBaseMS
  (bind _out_sounds (Helper|GetSoundsfromEvent (select (CallFunction|IsLocallyViewed) "(TagName=\"SoundLibrary.Weapon.Fire.1P\")" "(TagName=\"SoundLibrary.Weapon.Fire.3P\")")))
  (if (> (Utilities|Array|Length _out_sounds) 0)
    (Utilities|IsValid (Utilities|Array|Get(acopy) _out_sounds)
      (:Is%20Valid
        (bind _asmetasound_source (Utilities|Casting|CastToMetaSoundSource (Utilities|Array|Get(acopy) _out_sounds)))
        (Variables|Default|SetBaseGunfireMS _asmetasound_source)))))

(fn SetModMS
  (bind _values (Utilities|Map|Values (Variables|FortWeaponSoundLibraryComponent|GetActiveMods)))
  (for _array_element _values
    (Utilities|IsValid _array_element
      (:Is%20Valid
        (Utilities|Array|AddUnique (Variables|Default|GetModMS) (select (CallFunction|IsLocallyViewed) (Utilities|ResolveSoftReference (Variables|/Script/FortniteGame.FortWeaponModAudioData|GetMetaSoundPatch1P _array_element)) (Utilities|ResolveSoftReference (Variables|/Script/FortniteGame.FortWeaponModAudioData|GetMetaSoundPatch3P _array_element)))))))
  (return))

(fn StopFireEventsOnDeath
  (params (DeadActor :type Actor%20Object%20Reference) (Damage :type Float%20%28single-precision%29) (InstigatedBy :type Fort%20Pawn%20Object%20Reference) (HitLocation :type Vector) (DamageCauser :type Actor%20Object%20Reference))
  (SoundLibrary|StopEvent "(TagName=\"SoundLibrary.Weapon.Fire.1P\")")
  (SoundLibrary|StopEvent "(TagName=\"SoundLibrary.Weapon.Fire.3P\")"))

(fn StopFireEventsOnUnequip
  (SoundLibrary|StopEvent "(TagName=\"SoundLibrary.Weapon.Fire.1P\")")
  (SoundLibrary|StopEvent "(TagName=\"SoundLibrary.Weapon.Fire.3P\")"))

(fn TryHandleFireEvent
  (params (Event :type Gameplay%20Tag%20Structure) (Component :type Audio%20Component%20Object%20Reference))
  (bind _returnvalue (Components|GetOwner))
  (bind _asfort_weapon (Utilities|Casting|CastToFortWeapon _returnvalue))
  (bind _returnvalue_2 (AudioParameter|GetAudioParameterComponent _asfort_weapon :!class /Script/FortniteGame.FortWeapon))
  (bind _asfort_audio_parameter_component_weapon (Utilities|Casting|CastToFortAudioParameterComponent_Weapon _returnvalue_2))
  (if (GameplayTags|MatchesTag Event (Variables|Default|GetFireEvent))
    (Utilities|IsValid _asfort_weapon
      (:Is%20Valid
        (Utilities|IsValid _returnvalue_2
          (:Is%20Valid
            (Utilities|IsValid _asfort_audio_parameter_component_weapon
              (:Is%20Valid
                (Audio|Parameter|SetBooleanParameter (Variables|GetComponent) "Weapon.IsSuppressorEnabled" (Variables|Default|GetbEnableSuppressor) :!class /Script/AudioExtensions.AudioParameterControllerInterface)
                (WeaponParameters|OnRateOfFireChanged _asfort_audio_parameter_component_weapon (FortWeapon|GetFiringRate _asfort_weapon))
                (WeaponParameters|OnWeaponFireAudio _asfort_audio_parameter_component_weapon _asfort_weapon))
              (:Is%20Not%20Valid
                _))
            (return))
          (:Is%20Not%20Valid
            _)))
      (:Is%20Not%20Valid
        _))
    (else
      _)))

(fn ValidateModMetasounds
  (params (1pMetaSoundPatch :type Object%20Reference) (3pMetaSoundPatch :type Object%20Reference))
  (if (CallFunction|IsLocallyViewed)
    (Utilities|IsValid 1pMetaSoundPatch
      (:Is%20Valid
        _)
      (:Is%20Not%20Valid
        (return false)))
    (else
      (Utilities|IsValid 3pMetaSoundPatch
        (:Is%20Valid
          _)
        (:Is%20Not%20Valid
          _))))
  (return true))
