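; Sound library context for weapons that plays weapon event sounds, handling fire sounds separately from the rest.
; Fire events either retrigger or spawn cached audio components on the weapon's owner, with queued indices pruned in reverse.
; A watched MetaSound output, bOnAlmostFinished, updates bAlmostFinished on spawned fire sounds.
; Other events spawn one-shot attached sounds, while a retrieval mode collects unique sounds instead of playing them.
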(fn CleanFireAudioComponents
  (bind _indextoclear (Variables|Default|GetIndexToClear))
  (if (> (Utilities|Array|Length _indextoclear) 0)
    (bind (_arrayindex _arrayelement) (Utilities|Array|ReverseforEachLoop _indextoclear)
      (:LoopBody
        (Utilities|Array|RemoveIndex (Variables|Default|GetFireAudioComponents) _arrayelement))
      (:Completed
        (Utilities|Array|Clear _indextoclear)
        (return false)))
    (else
      (return true))))

(fn ClearRetrievedSounds
  (Utilities|Array|Clear (Variables|Default|GetGatheredSounds))
  (Variables|Default|SetbGatherSounds false))

(fn EnableSoundRetrieval
  (Variables|Default|SetbGatherSounds true))

(fn GetRetrievedSounds
  (return (Variables|Default|GetGatheredSounds)))

(fn GetRootComponent
  (params (Actor :type Actor%20Object%20Reference) (EventTag :type Gameplay%20Tag%20Structure))
  (bind _returnvalue (Actor|GetOwner Actor))
  (bind _rootcomponent (Variables|/Script/Engine.Actor|GetRootComponent Actor))
  (if (GameplayTags|MatchesTag EventTag "(TagName=\"SoundLibrary.Weapon.Fire\")")
    (Utilities|IsValid _returnvalue
      (:Is%20Valid
        (return (Variables|/Script/Engine.Actor|GetRootComponent _returnvalue)))
      (:Is%20Not%20Valid
        (return _rootcomponent)))
    (else
      (return _rootcomponent))))

(fn OnAlmostFinished
  (params (OutputName :type Name) (Output :type Meta%20Sound%20Output%20Structure%20%28by%20ref%29))
  (bind (_success _returnvalue) (MetaSoundOutput|GetBool Output))
  (if _success
    (Variables|Default|SetbAlmostFinished _returnvalue)
    (return)))

(fn Play
  (params (InEventData :type Sound%20Library%20Context%20Event%20Input%20Structure%20%28by%20ref%29))
  (bind (_sound _owner _eventtag) (Utilities|Struct|BreakSoundLibraryContextEventInput InEventData))
  (bind _returnvalue (Weapon|GetAttachComponent _owner _eventtag))
  (bind _fireaudiocomponents (Variables|Default|GetFireAudioComponents))
  (if (Variables|Default|GetbGatherSounds)
    (Utilities|Array|AddUnique (Variables|Default|GetGatheredSounds) _sound)
    (elif (GameplayTags|MatchesTag _eventtag "(TagName=\"SoundLibrary.Weapon.Fire\")")
      (bind _fireaudiocomponents_2 (Utilities|ShouldSpawnFireSound _sound)
        (:OnReTrigger
          (return _fireaudiocomponents_2 true))
        (:SpawnAudioComponent
          (bind _audiocomponent (CallFunction|SetupAudioComponent _returnvalue _owner _sound))
          (MetaSoundOutput|WatchOutput (WorldSubsystems|GetMetaSoundOutputSubsystem) _audiocomponent "bOnAlmostFinished" (EventDispatchers|CreateEvent :!event OnAlmostFinished) "None" "None" :!class /Script/MetasoundEngine.MetaSoundOutputSubsystem)
          (Utilities|Array|AddUnique _fireaudiocomponents _audiocomponent)
          (return _fireaudiocomponents true)))
      (else
        (bind _returnvalue_2 (Audio|SpawnSoundAttached _sound _returnvalue "pelvis"))
        (return (Utilities|Array|MakeArray _returnvalue_2) true)))))

(fn SetupAudioComponent
  (params (AttachmentComponent :type Scene%20Component%20Object%20Reference) (Actor :type Actor%20Object%20Reference) (Sound :type Sound%20Base%20Object%20Reference))
  (bind _returnvalue (Actor|GetOwner Actor))
  (Utilities|IsValid _returnvalue
    (:Is%20Valid
      (bind _returnvalue_2 (Game|AddComponentbyClass _returnvalue "/Script/Engine.AudioComponent" true (Math|Transform|MakeTransform "10.000000,5.000000,60.000000")))
      (Transformation|AttachComponentToComponent _returnvalue_2 AttachmentComponent "None" "KeepRelative" "KeepRelative" "KeepRelative" true)
      (Audio|Components|Audio|SetSound _returnvalue_2 Sound)
      (Weapon|SetAutoDestroy _returnvalue_2)
      (return _returnvalue_2))))
