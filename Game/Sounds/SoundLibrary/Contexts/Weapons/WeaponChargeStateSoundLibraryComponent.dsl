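; Weapon sound library component that plays first- or third-person fire sounds by charge state (empty, partial, charged, overcharged).
; Weapon fire events take the weapon's charge state or a custom one from charge-percent thresholds or manual setting.
; It pushes fire rate, ammo, suppressor and aiming parameters to the MetaSound audio component, stopping sounds on unequip, death or EndPlay.
; The expanded third-person branch plays 1P tags, unlike the custom path.
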(event EventBeginPlay
  (CallFunction|InitializeEventBindings))

(event Custom|OnWeaponFired
  (params (Weapon :type Fort%20Weapon%20Object%20Reference))
  ; ChargeStateHandling
  (Variables|Advanced|GetCustomChargeStates
    (:then
      (CallFunction|OnWeaponFiredwithCustomChargeState Weapon))
    (:else
      (Variables|Default|SetChargeState (FortWeapon|GetWeaponChargeStateforFireFX Weapon))
      (bind _instigatorpawn (Variables|Default|GetInstigatorPawn)
        (:then
          (if (Pawn|IsLocallyViewed _instigatorpawn)
            (CallFunction|PlayFire1PSound)
            (else
              (CallFunction|PlayFire3PSound)))
          (CallFunction|UpdateParameterPack))
        (:else)))))

(event Custom|OnWeaponFiredWithCustomChargeState
  (params (Weapon :type Fort%20Weapon%20Object%20Reference))
  ; ChargeStateHandling
  (Variables|Default|GetUseWeaponChargePercentForCustomCharge
    (:then
      (CallFunction|DetermineCustomChargeState (FortWeapon|GetChargePercent Weapon)))
    (:else
      _))
  (Variables|Default|Get%20Weapon%20Sound%20Library%20Component
    (:then
      (bind _instigatorpawn (Variables|Default|GetInstigatorPawn)
        (:then
          (if (Pawn|IsLocallyViewed _instigatorpawn)
            (CallFunction|PlayCustomFire1P)
            (else
              (CallFunction|PlayCustomFire3P)))
          (CallFunction|UpdateParameterPack))
        (:else)))
    (:else)))

(event Custom|OnOneShotFireSFX
  ; FireHandling
  (bind _weaponfireaudiocomponent (Variables|Default|GetWeaponFireAudioComponent)
    (:then
      (Audio|Parameter|ExecuteTriggerParameter _weaponfireaudiocomponent "FireSingleShot")
      (bind _soundlibrarycontext (Variables|Default|GetSoundLibraryContext)
        (:then
          (Variables|/Game/Sounds/SoundLibrary/Contexts/Weapons/BP_FortWeaponSoundLibraryContext.BP_FortWeaponSoundLibraryContext_C|SetIs%20Almost%20Finished false _soundlibrarycontext))
        (:else)))
    (:else)))

(event Custom|DeathEffects
  (params (DeadActor :type Actor%20Object%20Reference) (Damage :type Float%20%28single-precision%29) (InstigatedBy :type Fort%20Pawn%20Object%20Reference) (HitLocation :type Vector) (DamageCauser :type Actor%20Object%20Reference))
  ; StopEventSafety
  (SoundLibrary|StopAllEvents))

(event Custom|OnWeaponUnequipped
  ; StopEventSafety
  (SoundLibrary|StopAllEvents))

(event EventEndPlay
  (params (EndPlayReason :type EEndPlayReason%20Enum :default "Destroyed"))
  ; StopEventSafety
  (SoundLibrary|StopAllEvents))

(fn DetermineCustomChargeState
  (params (ChargePercent :type Float%20%28double-precision%29))
  (LocalVariables|SetPercentCharged ChargePercent)
  (if (< (LocalVariables|GetPercentCharged) (Variables|Advanced|GetChargeStatePercent_Partial))
    (Variables|Default|SetChargeState_CustomInt 0)
    (elif (< (LocalVariables|GetPercentCharged) (Variables|Advanced|GetChargeStatePercent_Charged))
      (Variables|Default|SetChargeState_CustomInt 1)
      (elif (< (LocalVariables|GetPercentCharged) (Variables|Advanced|GetChargeStatePercent_OverCharged))
        (Variables|Default|SetChargeState_CustomInt 2)
        (else
          (Variables|Default|SetChargeState_CustomInt 3))))))

(fn InitializeEventBindings
  (bind _returnvalue (Components|GetOwner))
  (Variables|Default|SetWeapon _returnvalue)
  (Variables|Default|SetInstigatorPawn (Variables|/Script/Engine.Actor|GetInstigator _returnvalue))
  (bind _asfort_weapon (Utilities|Casting|CastToFortWeapon _returnvalue))
  (Utilities|FlowControl|Sequence
    (:then_0
      (Variables|Default|SetFortWeapon _asfort_weapon)
      (EventDispatchers|BindEventtoOnWeaponFireAudio _asfort_weapon (EventDispatchers|CreateEvent :!event OnWeaponFired))
      (EventDispatchers|BindEventtoOnLocalTargetingChanged _asfort_weapon (EventDispatchers|CreateEvent :!event IsAimingDownSights))
      (FortWeapon|BindEventtoOnWeaponRateOfFireChanged _asfort_weapon (EventDispatchers|CreateEvent :!event OnRateOfFireChanged)))
    (:then_1
      (bind _instigatorpawn (Variables|Default|GetInstigatorPawn)
        (:then
          (bind _asfort_player_pawn_athena (Utilities|Casting|CastToFortPlayerPawnAthena _instigatorpawn))
          (FortPawn|BindEventtoOnWeaponUnEquippedDelegate _asfort_player_pawn_athena (EventDispatchers|CreateEvent :!event OnWeaponUnequipped))
          (EventDispatchers|BindEventtoOnDeathEffects _asfort_player_pawn_athena (EventDispatchers|CreateEvent :!event DeathEffects)))
        (:else)))
    (:then_2
      (bind _asbp_fort_weapon_charged_sound_library_context (Utilities|Casting|CastToBP_FortWeapon_Charged_SoundLibraryContext (SoundLibrary|GetContext)))
      (Variables|Default|SetSoundLibraryContext _asbp_fort_weapon_charged_sound_library_context))))

(fn IsAimingDownSights
  (params (bIsTargeting :type Boolean))
  (bind _output_get (Variables|Parameters|SetIsADS bIsTargeting))
  (bind _weaponfireaudiocomponent (Variables|Default|GetWeaponFireAudioComponent)
    (:then
      (Audio|Parameter|SetBooleanParameter _weaponfireaudiocomponent "Weapon.IsAimingDownSights" _output_get :!class /Script/AudioExtensions.AudioParameterControllerInterface))
    (:else)))

(fn OnRateOfFireChanged
  (params (NewRateOfFire :type Float%20%28single-precision%29))
  (bind _weaponfireaudiocomponent (Variables|Default|GetWeaponFireAudioComponent)
    (:then
      (Audio|Parameter|SetFloatParameter _weaponfireaudiocomponent "Weapon.FireRate" NewRateOfFire :!class /Script/AudioExtensions.AudioParameterControllerInterface))
    (:else)))

(fn OnSoundPlayed
  (params (InEventName :type Gameplay%20Tag%20Structure) (InComponent :type Audio%20Component%20Object%20Reference))
  (|CallParentFunction InEventName InComponent)
  (bind _output_get (Variables|Default|SetEventName InEventName))
  (if (GameplayTags|MatchesAnyTags _output_get (Variables|Default|GetChargedStateTags))
    (Variables|Default|SetWeaponFireAudioComponent (Variables|GetInComponent))
    (CallFunction|OnOneShotFireSFX)))

(fn OnSoundStopped
  (params (InEventName :type Gameplay%20Tag%20Structure) (InComponent :type Audio%20Component%20Object%20Reference))
  (if (GameplayTags|MatchesTag InEventName "(TagName=\"SoundLibrary.Weapon.Fire\")")
    (if (GameplayTags|MatchesAnyTags (Variables|GetInEventName) (Variables|Default|Get1PChargeTags))
      (bind _incomponent (Variables|GetInComponent)
        (:then
          (Audio|Parameter|SetBooleanParameter _incomponent "bIsFiring" false :!class /Script/AudioExtensions.AudioParameterControllerInterface)
          (return true))
        (:else
          (return true)))
      (elif (GameplayTags|MatchesAnyTags (Variables|GetInEventName) (Variables|Default|Get3PChargeTags))
        (bind _incomponent_2 (Variables|GetInComponent)
          (:then
            (Audio|Parameter|ExecuteTriggerParameter _incomponent_2 "StopFiring"))
          (:else
            (return true)))))
    (else
      (return false))))

(fn PlayFire1PSound
  (Variables|Default|GetUseEmptyAndOverchargeValues
    (:then
      ; Expanded Implementation
      (switch Utilities|FlowControl|Switch|SwitchonEFortWeaponChargeStateForFireFX (Variables|Default|GetChargeState)
        (:Partial
          (SoundLibrary|PlayEvent "(TagName=\"SoundLibrary.Weapon.Fire.Partial.1P\")"))
        (:Full
          (SoundLibrary|PlayEvent "(TagName=\"SoundLibrary.Weapon.Fire.Charged.1P\")"))
        (:Over
          (SoundLibrary|PlayEvent "(TagName=\"SoundLibrary.Weapon.Fire.Charged.Over.1P\")"))
        (:Max_None
          (SoundLibrary|PlayEvent "(TagName=\"SoundLibrary.Weapon.Fire.Empty.1P\")"))))
    (:else
      ; Default Charge Implementation
      (switch Utilities|FlowControl|Switch|SwitchonEFortWeaponChargeStateForFireFX (Variables|Default|GetChargeState)
        (:Partial
          _)
        (:Full
          (SoundLibrary|PlayEvent "(TagName=\"SoundLibrary.Weapon.Fire.Charged.1P\")"))
        (:Over
          _)
        (:Max_None
          _))
      (SoundLibrary|PlayEvent "(TagName=\"SoundLibrary.Weapon.Fire.Partial.1P\")"))))

(fn PlayFire3PSound
  (Variables|Default|GetUseEmptyAndOverchargeValues
    (:then
      ; Expanded Implementation
      (switch Utilities|FlowControl|Switch|SwitchonEFortWeaponChargeStateForFireFX (Variables|Default|GetChargeState)
        (:Partial
          (SoundLibrary|PlayEvent "(TagName=\"SoundLibrary.Weapon.Fire.Partial.1P\")"))
        (:Full
          (SoundLibrary|PlayEvent "(TagName=\"SoundLibrary.Weapon.Fire.Charged.1P\")"))
        (:Over
          (SoundLibrary|PlayEvent "(TagName=\"SoundLibrary.Weapon.Fire.Charged.Over.1P\")"))
        (:Max_None
          (SoundLibrary|PlayEvent "(TagName=\"SoundLibrary.Weapon.Fire.Empty.1P\")"))))
    (:else
      ; Default Implementation
      (switch Utilities|FlowControl|Switch|SwitchonEFortWeaponChargeStateForFireFX (Variables|Default|GetChargeState)
        (:Partial
          _)
        (:Full
          (SoundLibrary|PlayEvent "(TagName=\"SoundLibrary.Weapon.Fire.Charged.3P\")"))
        (:Over
          _)
        (:Max_None
          _))
      (SoundLibrary|PlayEvent "(TagName=\"SoundLibrary.Weapon.Fire.Partial.3P\")"))))

(fn PlayCustomFire1P
  (SoundLibrary|PlayEvent (Utilities|Select :Index (Variables|Default|GetChargeState_CustomInt) :Option%200 "(TagName=\"SoundLibrary.Weapon.Fire.Empty.1P\")" :Option%201 "(TagName=\"SoundLibrary.Weapon.Fire.Partial.1P\")" :Option%202 "(TagName=\"SoundLibrary.Weapon.Fire.Charged.1P\")" :Option%203 "(TagName=\"SoundLibrary.Weapon.Fire.Charged.Over.1P\")")))

(fn PlayCustomFire3P
  (SoundLibrary|PlayEvent (Utilities|Select :Index (Variables|Default|GetChargeState_CustomInt) :Option%200 "(TagName=\"SoundLibrary.Weapon.Fire.Empty.3P\")" :Option%201 "(TagName=\"SoundLibrary.Weapon.Fire.Partial.3P\")" :Option%202 "(TagName=\"SoundLibrary.Weapon.Fire.Charged.3P\")" :Option%203 "(TagName=\"SoundLibrary.Weapon.Fire.Charged.Over.3P\")")))

(fn SetCustomChargeState
  (params (ChargeState_CustomInt :type Integer))
  (Variables|Default|SetChargeState_CustomInt ChargeState_CustomInt)
  (Variables|Default|SetUseWeaponChargePercentForCustomCharge false))

(fn UpdateParameterPack
  ; Necessary to Update the Parameters in the Weapon Metasound Interface. Hopefully to be deprecated when we can move to childed Weapon Sound Library Components so we don't have parellel design paths.
  (bind _weaponfireaudiocomponent (Variables|Default|GetWeaponFireAudioComponent)
    (:then
      (Audio|Parameter|SetParameters _weaponfireaudiocomponent (Utilities|Array|MakeArray (Audio|Parameter|FloattoAudioParameter "Weapon.FireRate" (Math|Float|SafeDivide 1.0 (FortWeapon|GetFiringRate (Variables|Default|GetFortWeapon)))) (Audio|Parameter|IntegertoAudioParameter "Weapon.AmmoRemaining" (FortWeapon|GetMagazineAmmoCount (Variables|Default|GetFortWeapon))) (Audio|Parameter|BooleantoAudioParameter "Weapon.IsSuppressorEnabled" (FortWeapon|IsSuppressed (Variables|Default|GetFortWeapon))) (Audio|Parameter|BooleantoAudioParameter "Weapon.IsAimingDownSights" (FortWeapon|IsTargeting (Variables|Default|GetFortWeapon))))))
    (:else)))
